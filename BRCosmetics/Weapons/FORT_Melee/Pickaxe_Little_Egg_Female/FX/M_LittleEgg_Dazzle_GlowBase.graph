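# Additive unlit base material for the Little Egg pickaxe's dazzle glow, usable on Niagara sprites and mesh particles.
# Procedural sphere masks build a sharp, partly desaturated center over a soft halo, tinted by particle color hue-shifted by distance from the actor.
# Particle alpha passes through depth fade (skipped on mobile), with switchable view-angle and near-camera fades.
# A switchable "Use Distance Fade" branch computes a vertex-scale offset that grows with camera distance.
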Material /BRCosmetics/Weapons/FORT_Melee/Pickaxe_Little_Egg_Female/FX/M_LittleEgg_Dazzle_GlowBase
graph {
 properties {
  bAllowNegativeEmissiveColor=true
  bUsedWithNiagaraMeshParticles=true
  bUsedWithNiagaraSprites=true
  blendMode=BLEND_Additive
  pixelDepthOffsetMode=PDOM_Legacy
  shadingModel=MSM_Unlit
  usedShadingModels=Unlit
 }
 nodes [
  n1 = TextureCoordinate()
  n2 = Constant() {r=0.5}
  n3 = Multiply(
   A=ScalarParameter() {defaultValue=2, parameterName=CenterBrightness}
   B=Power(Base=SphereMask(A=@n1, B=@n2, Radius=@n2) {hardnessPercent=0}, Exp=ScalarParameter() {defaultValue=25, parameterName=CentertPower}) {constExponent=25}
  ) {constB=2}
  n4 = Desaturation(None=NamedRerouteUsage(), Fraction=ScalarParameter() {defaultValue=0.3, parameterName=CenterDesaturation})
  n5 = TextureCoordinate()
  n6 = SphereMask(A=@n5, B=Constant() {r=0.5}) {attenuationRadius=0.5, hardnessPercent=0}
  n7 = NamedRerouteUsage()
  MaterialFunctionCall(Emissive=Add(A=Multiply(A=@n3, B=@n4), B=Multiply(A=@n6, B=@n7))) {materialFunction=&MF_FX_EmissiveAdjustments}
  Multiply(B=@n6)
  n8 = MaterialFunctionCall("FadeSensitivity (optional)"=ScalarParameter() {defaultValue=1, parameterName=AngleFade}) {materialFunction=&FadeBasedOnViewAngle}
  n9 = Constant() {r=1}
  n10 = ParticleColor()
  n11 = DynamicParameter() {desc=Index 1 Params for Niagara, parameterIndex=1, paramNames [DepthFadeDist, Param2, Param3, Param4]}
  n12 = ScalarParameter() {defaultValue=30, parameterName=DepthFade}
  n13 = DepthFade(Opacity=@n10.A, FadeDistance=If(A=@n11.DepthFadeDist, B=Constant() {r=1}, "A > B"=@n11.DepthFadeDist, "A == B"=@n12, "A < B"=@n12))
  n14 = Multiply(
   A=Multiply(A=StaticSwitchParameter(True=@n8.Result, False=@n9) {parameterName=Use Angle Fade}, B=Constant() {r=1})
   B=ShadingPathSwitch(Default=@n13, Deferred=@n13, Mobile=@n10.A) {desc=Migrated from FeatureLevelSwitch}
  )
  n15 = MaterialFunctionCall(Distance=ScalarParameter() {parameterName=CameraFadeDistance}) {materialFunction=&Near_Camera_Fade}
  Multiply(
   A=StaticSwitchParameter(True=Multiply(A=@n14, B=@n15.Result), False=@n14) {group=DetailedGradientMask, parameterName=UseCameraNearFade}
   B=@n6
  )
  n16 = WorldPosition()
  n17 = ActorPositionWS()
  n18 = OneMinus(None=MaterialFunctionCall(Vector 3=Subtract(A=@n16.XYZ, B=@n17)) {materialFunction=/Engine/Functions/Engine_MaterialFunctions02/Utility/VectorLength}.V3 Length)
  n19 = Divide(B=ScalarParameter() {parameterName=HueOffsetScale})
  NamedRerouteDeclaration(None=MaterialFunctionCall(Hue Shift Percentage=Multiply(A=@n18, B=@n19), Texture=ParticleColor().RGB) {materialFunction=&HueShift}.Result) {name=Color, nodeColor=<rgba 0.0547 1 0 1>}
  ParticlePositionWS()
  n20 = ScalarParameter() {defaultValue=10, parameterName=MaxScaleIncrease}
  n21 = ObjectPositionWS()
  n22 = Multiply(A=Subtract(A=@n21, B=WorldPosition().XYZ)) {constB=-1}
  n23 = CameraPositionWS()
  n24 = Divide(
   A=Subtract(A=Distance(A=@n21, B=@n23), B=ScalarParameter() {defaultValue=2000, parameterName=DistanceOffset})
   B=ScalarParameter() {defaultValue=30000, parameterName=ScaleUpDistance}
  )
  StaticSwitchParameter(True=LinearInterpolate(B=Multiply(A=@n20, B=@n22) {constB=-1}, Alpha=Clamp(None=@n24)) {constB=0}, False=Constant()) {parameterName=Use Distance Fade}
  n25 = VectorParameter() {defaultValue=<rgba 0 0 0 1>}
  TextureSampleParameter2D(UVs=Add(A=@n5, B=ComponentMask(None=@n25.RGB) {g=true, r=true})) {parameterName=FlareMask, texture=/Game/Effects/Fort_Effects/Textures/Flares/T_Flare_Soft}
 ]
}
