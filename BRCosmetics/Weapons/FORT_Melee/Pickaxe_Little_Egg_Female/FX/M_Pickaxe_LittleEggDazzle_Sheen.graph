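# Additive unlit sheen effect material for the Little Egg Dazzle female pickaxe, used on Niagara mesh particles.
# The pickaxe diffuse texture is hue-shifted by second-UV U, panning noise and the HueShift dynamic parameter, tinted by particle color and fed into an emissive-adjustment function.
# A clamped gradient is sampled at second UVs offset by the SheenPosition dynamic parameter and distorted by the same noise, which pans over particle relative time.
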Material /BRCosmetics/Weapons/FORT_Melee/Pickaxe_Little_Egg_Female/FX/M_Pickaxe_LittleEggDazzle_Sheen
graph {
 properties {
  bUsedWithNiagaraMeshParticles=true
  blendMode=BLEND_Additive
  pixelDepthOffsetMode=PDOM_Legacy
  previewMesh=&/SM_FX_LittleEgg_F_Axe_Sheen
  shadingModel=MSM_Unlit
  usedShadingModels=Unlit
 }
 nodes [
  Add(A=TextureCoordinate(), B=Constant2Vector() {g=0.1, r=0.1})
  Frac(None=Time())
  n1 = DynamicParameter() {defaultValue=<rgba 0 1 1 1>, paramNames [SheenPosition, HueShift, Param3, Param4]}
  n2 = TextureSample(UVs=Panner(Coordinate=TextureCoordinate() {coordinateIndex=1, uTiling=0.5, vTiling=0.5}, Time=ParticleRelativeTime()) {bFractionalPart=true, speedX=0.25, speedY=0.5}) {samplerType=SAMPLERTYPE_LinearGrayscale, texture=/Game/Effects/Fort_Effects/Textures/Tile/T_Diagonal_Blur_Noise_M}
  n3 = Add(A=ComponentMask(None=TextureCoordinate() {coordinateIndex=1}) {r=true}, B=Add(A=@n1.HueShift, B=Multiply(B=@n2.G) {constA=0.5}))
  n4 = TextureSample() {texture=/BRCosmetics/Weapons/FORT_Melee/Pickaxe_Little_Egg_Female/Skins/Textures/T_Pickaxe_Little_Egg_Dazzle_Female_D}
  n5 = AppendVector(
   A=AppendVector(
    A=@n4.R
    B=TextureSample() {texture=&T_Pickaxe_Little_Egg_Dazzle_Female_D}.G
   )
   B=@n4.B
  )
  MaterialFunctionCall(Emissive=Multiply(
    A=ParticleColor().RGB
    B=MaterialFunctionCall(Hue Shift Percentage=@n3, Texture=@n5) {materialFunction=&HueShift}.Result
   )) {materialFunction=&MF_FX_EmissiveAdjustments}
  TextureCoordinate() {coordinateIndex=1, uTiling=0.5, vTiling=0.5}
  n6 = TextureCoordinate() {coordinateIndex=1}
  TextureSample(UVs=Add(A=Add(A=@n6, B=@n1.SheenPosition), B=Multiply(A=@n2.G) {constB=0.2})) {samplerSource=SSM_Clamp_WorldGroupSettings, texture=/Game/Effects/Fort_Effects/Textures/Grad/gradient_healed}
 ]
}
